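# Sound cue for the default demolition electric-zap impact, routed to the Fort_Impacts sound class.
# A quality-level root has two near-identical populated branches and two empty slots.
# Each branch plays the same Tesla Coil pickaxe electricity wave through a short random delay, modulator, attenuation and mixers; other mixer layers are empty.
# Its assigned concurrency rule, by name, caps it at four voices, stopping furthest or oldest.
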SoundCue /Clyde/Sounds/Fort_Impact_Sounds/Demolition/Defaults/Demo_Default_Impact_ElecZap_MLTI_Cue
root=QualityLevel_0
nodes [
 {id=Attenuation_0, children [Modulator_1], properties {attenuationOverrides {falloffDistance=900, attenuationShapeExtents {x=100}}}}
 {id=Attenuation_1, children [Modulator_2], properties {attenuationOverrides {falloffDistance=900, attenuationShapeExtents {x=100}}}}
 {id=Attenuation_6, children [null], properties {attenuationOverrides {falloffDistance=900, attenuationShapeExtents {x=100}}}}
 {id=Attenuation_7, children [null], properties {attenuationOverrides {falloffDistance=1600}}}
 {id=Delay_0, children [WavePlayer_56], properties {delayMax=0.1}}
 {id=Delay_1, children [WavePlayer_8], properties {delayMax=0.1}}
 {id=Mixer_2, children [null, null, Attenuation_0], properties {inputVolume [1.2,0.7,1.25]}}
 {id=Mixer_3, children [SoundClass_3,null], properties {inputVolume [1,1]}}
 {id=Mixer_4, children [Attenuation_6, Attenuation_7, Attenuation_1], properties {inputVolume [1.2,0.7,1]}}
 {id=Mixer_5, children [SoundClass_5,null], properties {inputVolume [1,1]}}
 {id=Modulator_1, children [Delay_0]}
 {id=Modulator_2, children [Delay_1]}
 {id=QualityLevel_0, children [Mixer_3,Mixer_5,null,null]}
 {id=SoundClass_3, children [Mixer_2], properties {soundClassOverride=&Fort_Impacts}}
 {id=SoundClass_5, children [Mixer_4], properties {soundClassOverride=&Fort_Impacts}}
 {id=WavePlayer_56, properties {soundWaveAssetPtr=/Game/Sounds/Fort_Weapons/Impacts/Weapon_TeslaCoil_PickAxe_Impact_Electricity_55}}
 {id=WavePlayer_8, properties {soundWaveAssetPtr=&Weapon_TeslaCoil_PickAxe_Impact_Electricity_55}}
]
properties {
 maxDistance=2097152
 soundClassObject=&Fort_Impacts
 volumeMultiplier=0.875
 concurrencyOverrides {maxCount=4}
 concurrencySet [/Game/Sounds/Concurrency_Rules/Demo_FourMaximum_StopFurthestOldest_VolumeScale]
}
