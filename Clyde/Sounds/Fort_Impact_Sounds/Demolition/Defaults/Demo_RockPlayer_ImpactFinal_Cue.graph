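# Sound cue for the final player impact on rock in demolition, routed to the impacts sound class.
# A mixer blends three attenuated layers, each picking randomly from three waves through a modulator: rock impacts, stereo rock impacts and dedicated rock-final hits.
# The stereo layer starts after a short random delay, and it and the final-hit layer sit lower in the mix with their own attenuation settings.
# A shared object-impact concurrency asset governs simultaneous plays.
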SoundCue /Clyde/Sounds/Fort_Impact_Sounds/Demolition/Defaults/Demo_RockPlayer_ImpactFinal_Cue
root=Mixer_3
nodes [
 {id=Attenuation_0, children [Modulator_1]}
 {id=Attenuation_3, children [Delay_0], properties {attenuationOverrides {falloffDistance=900, attenuationShapeExtents {x=100}}}}
 {id=Attenuation_4, children [Modulator_3], properties {attenuationOverrides {falloffDistance=900, attenuationShapeExtents {x=100}}}}
 {id=Delay_0, children [Modulator_0], properties {delayMax=0.1, delayMin=0.05}}
 {id=Mixer_3, children [Attenuation_0, Attenuation_3, Attenuation_4], properties {inputVolume [0.9,0.4,0.5]}}
 {id=Modulator_0, children [Random_1]}
 {id=Modulator_1, children [Random_0]}
 {id=Modulator_3, children [Random_5]}
 {id=Random_0, children [WavePlayer_239, WavePlayer_238, WavePlayer_245], properties {weights [1,1,1]}}
 {id=Random_1, children [WavePlayer_249, WavePlayer_252, WavePlayer_253], properties {weights [1,1,1]}}
 {id=Random_5, children [WavePlayer_257, WavePlayer_258, WavePlayer_259], properties {weights [1,1,1]}}
 {id=WavePlayer_238, properties {soundWaveAssetPtr=/Clyde/Sounds/Fort_Impact_Sounds/Demolition/rock_impact_05}}
 {id=WavePlayer_239, properties {soundWaveAssetPtr=/Clyde/Sounds/Fort_Impact_Sounds/Demolition/rock_impact_04}}
 {id=WavePlayer_245, properties {soundWaveAssetPtr=/Clyde/Sounds/Fort_Impact_Sounds/Demolition/rock_impact_07}}
 {id=WavePlayer_249, properties {soundWaveAssetPtr=/Game/Sounds/Fort_Impact_Sounds/Demolition/Defaults/rock_impact_stereo_03}}
 {id=WavePlayer_252, properties {soundWaveAssetPtr=/Game/Sounds/Fort_Impact_Sounds/Demolition/Defaults/rock_impact_stereo_06}}
 {id=WavePlayer_253, properties {soundWaveAssetPtr=/Game/Sounds/Fort_Impact_Sounds/Demolition/Defaults/rock_impact_stereo_07}}
 {id=WavePlayer_257, properties {soundWaveAssetPtr=/Clyde/Sounds/Fort_Impact_Sounds/Demolition/Fort_Impact_RockFinal_01}}
 {id=WavePlayer_258, properties {soundWaveAssetPtr=/Clyde/Sounds/Fort_Impact_Sounds/Demolition/Fort_Impact_RockFinal_02}}
 {id=WavePlayer_259, properties {soundWaveAssetPtr=/Clyde/Sounds/Fort_Impact_Sounds/Demolition/Fort_Impact_RockFinal_03}}
]
properties {
 maxDistance=2097152
 soundClassObject=&Fort_Impacts_PSM
 volumeMultiplier=0.425
 concurrencyOverrides {resolutionRule=PreventNew}
 concurrencySet [&CONC_Object_Impacts_FourMax]
}
