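# Sound cue for the final rock impact of Fort demolition sounds, routed to the Fort_Impacts_PSM sound class.
# A mixer layers a fixed stone impact, a briefly delayed random pick of three rock impacts, and a modulated random pick of three stereo rock impacts.
# Each layer passes through its own attenuation node, and a shared concurrency rule, named for a four-voice limit stopping the furthest or oldest, governs overlapping plays.
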SoundCue /Clyde/Sounds/Fort_Impact_Sounds/Demolition/Defaults/Demo_Rock_ImpactFinal_Cue
root=Mixer_3
nodes [
 {id=Attenuation_0, children [WavePlayer_6]}
 {id=Attenuation_3, children [Delay_0], properties {attenuationOverrides {falloffDistance=900, attenuationShapeExtents {x=100}}}}
 {id=Attenuation_4, children [Modulator_3], properties {attenuationOverrides {falloffDistance=900, attenuationShapeExtents {x=100}}}}
 {id=Delay_0, children [Random_5], properties {delayMax=0.1, delayMin=0.05}}
 {id=Mixer_3, children [Attenuation_0, Attenuation_3, Attenuation_4], properties {inputVolume [1,1,1]}}
 {id=Modulator_3, children [Random_1]}
 {id=Random_1, children [WavePlayer_9,WavePlayer_4,WavePlayer_5], properties {weights [1,1,1]}}
 {id=Random_5, children [WavePlayer_1, WavePlayer_64, WavePlayer_10], properties {weights [1,1,1]}}
 {id=WavePlayer_1, properties {soundWaveAssetPtr=/Clyde/Sounds/Fort_Impact_Sounds/Demolition/Fort_Impact_RockFinal_01}}
 {id=WavePlayer_10, properties {soundWaveAssetPtr=/Game/Sounds/Fort_Impact_Sounds/Demolition/Defaults/rock_impact_stereo_03}}
 {id=WavePlayer_4, properties {soundWaveAssetPtr=/Game/Sounds/Fort_Impact_Sounds/Demolition/Defaults/rock_impact_stereo_06}}
 {id=WavePlayer_5, properties {soundWaveAssetPtr=/Game/Sounds/Fort_Impact_Sounds/Demolition/Defaults/rock_impact_stereo_07}}
 {id=WavePlayer_6, properties {soundWaveAssetPtr=/Game/Sounds/Fort_Impact_Sounds/Demolition/Defaults/default_stone_impact_new_combined_04}}
 {id=WavePlayer_64, properties {soundWaveAssetPtr=/Clyde/Sounds/Fort_Impact_Sounds/Demolition/Fort_Impact_RockFinal_02}}
 {id=WavePlayer_9, properties {soundWaveAssetPtr=&rock_impact_stereo_03}}
]
properties {
 maxDistance=2097152
 soundClassObject=&Fort_Impacts_PSM
 volumeMultiplier=0.625
 concurrencyOverrides {maxCount=1}
 concurrencySet [/Game/Sounds/Concurrency_Rules/FourMaximum_StopFurthestOldest_VolumeScale]
}
