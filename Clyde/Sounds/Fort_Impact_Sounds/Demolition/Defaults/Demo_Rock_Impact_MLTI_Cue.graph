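# Sound cue for rock demolition impacts, in the Fort_Impacts sound class with concurrency limited to four.
# Each quality-level branch mixes three layers: stone impact and rock impact with short-range attenuation, plus a distant rock impact with a larger radius and falloff.
# The first two branches draw from random wave pools, the second with fewer stone variants, while the last two use one fixed wave per layer.
# Per-layer modulators vary pitch and set volume.
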SoundCue /Clyde/Sounds/Fort_Impact_Sounds/Demolition/Defaults/Demo_Rock_Impact_MLTI_Cue
root=QualityLevel_0
nodes [
 {id=Attenuation_0, children [Modulator_1], properties {attenuationOverrides {falloffDistance=3000, attenuationShapeExtents {x=1000}}}}
 {id=Attenuation_1, children [Modulator_5], properties {attenuationOverrides {falloffDistance=900, attenuationShapeExtents {x=100}}}}
 {id=Attenuation_10, children [Modulator_11], properties {attenuationOverrides {falloffDistance=900, attenuationShapeExtents {x=100}}}}
 {id=Attenuation_11, children [Modulator_9], properties {attenuationOverrides {falloffDistance=900, attenuationShapeExtents {x=100}}}}
 {id=Attenuation_2, children [Modulator_3], properties {attenuationOverrides {falloffDistance=900, attenuationShapeExtents {x=100}}}}
 {id=Attenuation_3, children [Modulator_4], properties {attenuationOverrides {falloffDistance=3000, attenuationShapeExtents {x=1000}}}}
 {id=Attenuation_4, children [Modulator_2], properties {attenuationOverrides {falloffDistance=900, attenuationShapeExtents {x=100}}}}
 {id=Attenuation_5, children [Modulator_7], properties {attenuationOverrides {falloffDistance=3000, attenuationShapeExtents {x=1000}}}}
 {id=Attenuation_6, children [Modulator_0], properties {attenuationOverrides {falloffDistance=900, attenuationShapeExtents {x=100}}}}
 {id=Attenuation_7, children [Modulator_10], properties {attenuationOverrides {falloffDistance=3000, attenuationShapeExtents {x=1000}}}}
 {id=Attenuation_8, children [Modulator_8], properties {attenuationOverrides {falloffDistance=900, attenuationShapeExtents {x=100}}}}
 {id=Attenuation_9, children [Modulator_6], properties {attenuationOverrides {falloffDistance=900, attenuationShapeExtents {x=100}}}}
 {id=Mixer_0, children [Attenuation_4, Attenuation_6], properties {inputVolume [1.2,0.7]}}
 {id=Mixer_1, children [Mixer_0, Attenuation_0], properties {inputVolume [1,1]}}
 {id=Mixer_2, children [Attenuation_1, Attenuation_2], properties {inputVolume [1.2,0.4]}}
 {id=Mixer_3, children [Mixer_2, Attenuation_3], properties {inputVolume [1,0.8]}}
 {id=Mixer_4, children [Attenuation_8, Attenuation_9], properties {inputVolume [1.2,0.7]}}
 {id=Mixer_5, children [Mixer_4, Attenuation_5], properties {inputVolume [1,1]}}
 {id=Mixer_6, children [Attenuation_10, Attenuation_11], properties {inputVolume [1.2,0.7]}}
 {id=Mixer_7, children [Mixer_6, Attenuation_7], properties {inputVolume [1,1]}}
 {id=Modulator_0, children [WavePlayer_4], properties {pitchMax=1.2, pitchMin=0.8, volumeMax=2, volumeMin=2}}
 {id=Modulator_1, children [WavePlayer_1], properties {pitchMax=1.4}}
 {id=Modulator_10, children [WavePlayer_15], properties {pitchMax=1.4}}
 {id=Modulator_11, children [WavePlayer_17], properties {pitchMax=1.1, pitchMin=0.9}}
 {id=Modulator_2, children [WavePlayer_8], properties {pitchMax=1.1, pitchMin=0.9}}
 {id=Modulator_3, children [Random_4], properties {pitchMax=1.2, pitchMin=0.8, volumeMax=2, volumeMin=2}}
 {id=Modulator_4, children [Random_3], properties {pitchMax=1.4}}
 {id=Modulator_5, children [Random_5]}
 {id=Modulator_6, children [Random_7], properties {pitchMax=1.2, pitchMin=0.8, volumeMax=2, volumeMin=2}}
 {id=Modulator_7, children [Random_6], properties {pitchMax=1.4}}
 {id=Modulator_8, children [Random_8]}
 {id=Modulator_9, children [WavePlayer_16], properties {pitchMax=1.2, pitchMin=0.8, volumeMax=2, volumeMin=2}}
 {id=QualityLevel_0, children [Mixer_3,Mixer_5,Mixer_1,Mixer_7]}
 {id=Random_3, children [WavePlayer_19, WavePlayer_18, WavePlayer_20, WavePlayer_21], properties {weights [1,1,1,1]}}
 {id=Random_4, children [WavePlayer_22, WavePlayer_23, WavePlayer_24, WavePlayer_25], properties {weights [1,1,1,1]}}
 {
  id=Random_5
  children [
   WavePlayer_26
   WavePlayer_27
   WavePlayer_28
   WavePlayer_29
   WavePlayer_30
   WavePlayer_31
   WavePlayer_32
   WavePlayer_33
   WavePlayer_34
   WavePlayer_35
  ]
  properties {weights [1,1,1,1,1,1,1,1,1,1]}
 }
 {id=Random_6, children [WavePlayer_37, WavePlayer_36, WavePlayer_38, WavePlayer_39], properties {weights [1,1,1,1]}}
 {id=Random_7, children [WavePlayer_40, WavePlayer_41, WavePlayer_42, WavePlayer_43], properties {weights [1,1,1,1]}}
 {
  id=Random_8
  children [WavePlayer_44, WavePlayer_45, WavePlayer_46, WavePlayer_48, WavePlayer_49, WavePlayer_50, WavePlayer_53]
  properties {weights [1,1,1,1,1,1,1]}
 }
 {id=WavePlayer_1, properties {soundWaveAssetPtr=/Clyde/Sounds/Fort_Impact_Sounds/Demolition/Rock_Distant_Impact_02}}
 {id=WavePlayer_15, properties {soundWaveAssetPtr=&Rock_Distant_Impact_02}}
 {id=WavePlayer_16, properties {soundWaveAssetPtr=/Clyde/Sounds/Fort_Impact_Sounds/Demolition/rock_newimp_03}}
 {id=WavePlayer_17, properties {soundWaveAssetPtr=/Game/Sounds/Fort_Impact_Sounds/Demolition/Defaults/default_stone_impact_new_combined_10}}
 {id=WavePlayer_18, properties {soundWaveAssetPtr=/Clyde/Sounds/Fort_Impact_Sounds/Demolition/Rock_Distant_Impact_10}}
 {id=WavePlayer_19, properties {soundWaveAssetPtr=&Rock_Distant_Impact_02}}
 {id=WavePlayer_20, properties {soundWaveAssetPtr=/Clyde/Sounds/Fort_Impact_Sounds/Demolition/Rock_Distant_Impact_03}}
 {id=WavePlayer_21, properties {soundWaveAssetPtr=/Clyde/Sounds/Fort_Impact_Sounds/Demolition/Rock_Distant_Impact_07}}
 {id=WavePlayer_22, properties {soundWaveAssetPtr=&rock_newimp_03}}
 {id=WavePlayer_23, properties {soundWaveAssetPtr=/Clyde/Sounds/Fort_Impact_Sounds/Demolition/rock_newimp_11}}
 {id=WavePlayer_24, properties {soundWaveAssetPtr=/Clyde/Sounds/Fort_Impact_Sounds/Demolition/rock_newimp_12}}
 {id=WavePlayer_25, properties {soundWaveAssetPtr=/Clyde/Sounds/Fort_Impact_Sounds/Demolition/rock_newimp_01}}
 {id=WavePlayer_26, properties {soundWaveAssetPtr=&default_stone_impact_new_combined_10}}
 {id=WavePlayer_27, properties {soundWaveAssetPtr=/Game/Sounds/Fort_Impact_Sounds/Demolition/Defaults/default_stone_impact_new_combined_01}}
 {id=WavePlayer_28, properties {soundWaveAssetPtr=/Game/Sounds/Fort_Impact_Sounds/Demolition/Defaults/default_stone_impact_new_combined_02}}
 {id=WavePlayer_29, properties {soundWaveAssetPtr=/Game/Sounds/Fort_Impact_Sounds/Demolition/Defaults/default_stone_impact_new_combined_03}}
 {id=WavePlayer_30, properties {soundWaveAssetPtr=/Game/Sounds/Fort_Impact_Sounds/Demolition/Defaults/default_stone_impact_new_combined_04}}
 {id=WavePlayer_31, properties {soundWaveAssetPtr=/Game/Sounds/Fort_Impact_Sounds/Demolition/Defaults/default_stone_impact_new_combined_05}}
 {id=WavePlayer_32, properties {soundWaveAssetPtr=/Game/Sounds/Fort_Impact_Sounds/Demolition/Defaults/default_stone_impact_new_combined_06}}
 {id=WavePlayer_33, properties {soundWaveAssetPtr=/Game/Sounds/Fort_Impact_Sounds/Demolition/Defaults/default_stone_impact_new_combined_07}}
 {id=WavePlayer_34, properties {soundWaveAssetPtr=/Game/Sounds/Fort_Impact_Sounds/Demolition/Defaults/default_stone_impact_new_combined_08}}
 {id=WavePlayer_35, properties {soundWaveAssetPtr=/Game/Sounds/Fort_Impact_Sounds/Demolition/Defaults/default_stone_impact_new_combined_09}}
 {id=WavePlayer_36, properties {soundWaveAssetPtr=&Rock_Distant_Impact_10}}
 {id=WavePlayer_37, properties {soundWaveAssetPtr=&Rock_Distant_Impact_02}}
 {id=WavePlayer_38, properties {soundWaveAssetPtr=&Rock_Distant_Impact_03}}
 {id=WavePlayer_39, properties {soundWaveAssetPtr=&Rock_Distant_Impact_07}}
 {id=WavePlayer_4, properties {soundWaveAssetPtr=&rock_newimp_03}}
 {id=WavePlayer_40, properties {soundWaveAssetPtr=&rock_newimp_03}}
 {id=WavePlayer_41, properties {soundWaveAssetPtr=&rock_newimp_11}}
 {id=WavePlayer_42, properties {soundWaveAssetPtr=&rock_newimp_12}}
 {id=WavePlayer_43, properties {soundWaveAssetPtr=&rock_newimp_01}}
 {id=WavePlayer_44, properties {soundWaveAssetPtr=&default_stone_impact_new_combined_10}}
 {id=WavePlayer_45, properties {soundWaveAssetPtr=&default_stone_impact_new_combined_01}}
 {id=WavePlayer_46, properties {soundWaveAssetPtr=&default_stone_impact_new_combined_02}}
 {id=WavePlayer_48, properties {soundWaveAssetPtr=&default_stone_impact_new_combined_04}}
 {id=WavePlayer_49, properties {soundWaveAssetPtr=&default_stone_impact_new_combined_05}}
 {id=WavePlayer_50, properties {soundWaveAssetPtr=&default_stone_impact_new_combined_06}}
 {id=WavePlayer_53, properties {soundWaveAssetPtr=&default_stone_impact_new_combined_09}}
 {id=WavePlayer_8, properties {soundWaveAssetPtr=&default_stone_impact_new_combined_10}}
]
properties {
 maxDistance=2097152
 soundClassObject=&Fort_Impacts
 volumeMultiplier=0.875
 concurrencyOverrides {maxCount=4}
 concurrencySet [/Game/Sounds/Concurrency_Rules/Demo_FourMaximum_StopFurthestOldest_VolumeScale]
}
