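# Sound cue for a rock sign impact in the demolition impact set, with melee-weapon attenuation and the impacts sound class.
# A quality-level node selects between two three-layer mixes, the second with fewer wave variations, and two branches playing one pre-optimized wave.
# Each mix combines two pitch-varied rock impact layers with a briefly delayed, higher-pitched stereo layer.
# Concurrency limits overlapping impacts through a shared concurrency set.
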SoundCue /Clyde/Sounds/Fort_Impact_Sounds/Demolition/Defaults/Demo_Rock_Sign_Impact_Cue
root=QualityLevel_0
nodes [
 {id=Delay_0, children [Modulator_0], properties {delayMax=0.08, delayMin=0.04}}
 {id=Delay_1, children [Modulator_1], properties {delayMax=0.08, delayMin=0.04}}
 {id=Mixer_0, children [Delay_1,Modulator_2,Modulator_3], properties {inputVolume [0.5,0.5,1]}}
 {id=Mixer_4, children [Delay_0,Modulator_39,Modulator_40], properties {inputVolume [0.5,0.5,1]}}
 {id=Modulator_0, children [Random_16], properties {pitchMax=1.6, pitchMin=1.4}}
 {id=Modulator_1, children [Random_1], properties {pitchMax=1.6, pitchMin=1.4}}
 {id=Modulator_2, children [Random_0], properties {pitchMax=1.2, pitchMin=0.8, volumeMax=2, volumeMin=2}}
 {id=Modulator_3, children [WavePlayer_7], properties {pitchMax=1.2, pitchMin=0.8, volumeMax=2, volumeMin=2}}
 {id=Modulator_39, children [Random_14], properties {pitchMax=1.2, pitchMin=0.8, volumeMax=2, volumeMin=2}}
 {id=Modulator_4, children [WavePlayer_0], properties {pitchMax=1.1, pitchMin=0.88}}
 {id=Modulator_40, children [Random_66], properties {pitchMax=1.2, pitchMin=0.8, volumeMax=2, volumeMin=2}}
 {id=Modulator_5, children [WavePlayer_9], properties {pitchMax=1.1, pitchMin=0.88}}
 {id=QualityLevel_0, children [Mixer_4,Mixer_0,Modulator_4,Modulator_5]}
 {id=Random_0, children [WavePlayer_4,WavePlayer_5], properties {weights [1,1]}}
 {id=Random_1, children [WavePlayer_1,WavePlayer_2], properties {bRandomizeWithoutReplacement=false, weights [1,1]}}
 {id=Random_14, children [WavePlayer_12, WavePlayer_14, WavePlayer_16], properties {weights [1,1,1]}}
 {id=Random_16, children [WavePlayer_177, WavePlayer_180, WavePlayer_181], properties {weights [1,1,1]}}
 {id=Random_66, children [WavePlayer_220, WavePlayer_221], properties {weights [1,1]}}
 {id=WavePlayer_0, properties {soundWaveAssetPtr=/Game/Sounds/Fort_Impact_Sounds/Demolition/Defaults/Rock_Sign_Impact_opt}}
 {id=WavePlayer_1, properties {soundWaveAssetPtr=/Game/Sounds/Fort_Impact_Sounds/Demolition/Defaults/rock_impact_stereo_03}}
 {id=WavePlayer_12, properties {soundWaveAssetPtr=/Clyde/Sounds/Fort_Impact_Sounds/Demolition/rock_newimp_03}}
 {id=WavePlayer_14, properties {soundWaveAssetPtr=/Clyde/Sounds/Fort_Impact_Sounds/Demolition/rock_newimp_11}}
 {id=WavePlayer_16, properties {soundWaveAssetPtr=/Clyde/Sounds/Fort_Impact_Sounds/Demolition/rock_newimp_01}}
 {id=WavePlayer_177, properties {soundWaveAssetPtr=&rock_impact_stereo_03}}
 {id=WavePlayer_180, properties {soundWaveAssetPtr=/Game/Sounds/Fort_Impact_Sounds/Demolition/Defaults/rock_impact_stereo_06}}
 {id=WavePlayer_181, properties {soundWaveAssetPtr=/Game/Sounds/Fort_Impact_Sounds/Demolition/Defaults/rock_impact_stereo_07}}
 {id=WavePlayer_2, properties {soundWaveAssetPtr=&rock_impact_stereo_06}}
 {id=WavePlayer_220, properties {soundWaveAssetPtr=/Clyde/Sounds/Fort_Impact_Sounds/Demolition/rock_impact_04}}
 {id=WavePlayer_221, properties {soundWaveAssetPtr=/Clyde/Sounds/Fort_Impact_Sounds/Demolition/rock_impact_05}}
 {id=WavePlayer_4, properties {soundWaveAssetPtr=&rock_newimp_03}}
 {id=WavePlayer_5, properties {soundWaveAssetPtr=&rock_newimp_11}}
 {id=WavePlayer_7, properties {soundWaveAssetPtr=&rock_impact_04}}
 {id=WavePlayer_9, properties {soundWaveAssetPtr=&Rock_Sign_Impact_opt}}
]
properties {
 attenuationSettings=&Fort_Weapon_Melee_Sml_Att
 maxDistance=2800
 soundClassObject=&Fort_Impacts
 volumeMultiplier=0.425
 concurrencyOverrides {maxCount=3, resolutionRule=StopOldest}
 concurrencySet [&CONC_Object_Impacts_FourMax]
}
