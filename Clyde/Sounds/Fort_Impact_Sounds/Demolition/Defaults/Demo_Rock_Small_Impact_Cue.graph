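# Sound cue for small rock demolition impacts, using small melee attenuation and the impacts sound class, with stop-oldest concurrency.
# A quality switch feeds four tiers: full, reduced-variation, and two that play a single optimized wave.
# The top two tiers mix a briefly delayed, pitched-up stereo rock hit, a quieter wobble and a boosted rock impact.
# Random nodes and modulators vary wave choice and pitch, with fewer waves in the second tier.
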SoundCue /Clyde/Sounds/Fort_Impact_Sounds/Demolition/Defaults/Demo_Rock_Small_Impact_Cue
root=QualityLevel_0
nodes [
 {id=Delay_0, children [Modulator_0], properties {delayMax=0.08, delayMin=0.04}}
 {id=Delay_1, children [Modulator_3], properties {delayMax=0.08, delayMin=0.04}}
 {id=Mixer_0, children [Delay_0,Modulator_2], properties {inputVolume [1.3,0.4]}}
 {id=Mixer_1, children [Delay_1,Modulator_5], properties {inputVolume [1.3,0.4]}}
 {id=Mixer_2, children [Mixer_1,Modulator_4], properties {inputVolume [1,1]}}
 {id=Mixer_4, children [Mixer_0,Modulator_1], properties {inputVolume [1,1]}}
 {id=Modulator_0, children [Random_16], properties {pitchMax=1.6, pitchMin=1.4}}
 {id=Modulator_1, children [Random_14], properties {pitchMax=1.2, pitchMin=0.8, volumeMax=2, volumeMin=2}}
 {id=Modulator_2, children [Random_4], properties {pitchMax=1.2}}
 {id=Modulator_3, children [Random_2], properties {pitchMax=1.6, pitchMin=1.4}}
 {id=Modulator_4, children [Random_1], properties {pitchMax=1.2, pitchMin=0.8, volumeMax=2, volumeMin=2}}
 {id=Modulator_5, children [WavePlayer_5], properties {pitchMax=1.2}}
 {id=Modulator_6, children [WavePlayer_0], properties {pitchMax=1.2, pitchMin=0.8}}
 {id=Modulator_7, children [WavePlayer_13], properties {pitchMax=1.2, pitchMin=0.8}}
 {id=QualityLevel_0, children [Mixer_4,Mixer_2,Modulator_6,Modulator_7]}
 {id=Random_1, children [WavePlayer_8,WavePlayer_9], properties {weights [1,1]}}
 {id=Random_14, children [WavePlayer_12, WavePlayer_14, WavePlayer_15, WavePlayer_16], properties {weights [1,1,1,1]}}
 {id=Random_16, children [WavePlayer_177, WavePlayer_180, WavePlayer_181, WavePlayer_178], properties {weights [1,1,1,1]}}
 {id=Random_2, children [WavePlayer_1,WavePlayer_3], properties {bRandomizeWithoutReplacement=false, weights [1,1]}}
 {id=Random_4, children [WavePlayer_185, WavePlayer_186, WavePlayer_187], properties {weights [1,1,1]}}
 {id=WavePlayer_0, properties {soundWaveAssetPtr=/Game/Sounds/Fort_Impact_Sounds/Demolition/Defaults/Rock_Small_Impact_opt}}
 {id=WavePlayer_1, properties {soundWaveAssetPtr=/Game/Sounds/Fort_Impact_Sounds/Demolition/Defaults/rock_impact_stereo_03}}
 {id=WavePlayer_12, properties {soundWaveAssetPtr=/Clyde/Sounds/Fort_Impact_Sounds/Demolition/rock_newimp_03}}
 {id=WavePlayer_13, properties {soundWaveAssetPtr=&Rock_Small_Impact_opt}}
 {id=WavePlayer_14, properties {soundWaveAssetPtr=/Clyde/Sounds/Fort_Impact_Sounds/Demolition/rock_newimp_11}}
 {id=WavePlayer_15, properties {soundWaveAssetPtr=/Clyde/Sounds/Fort_Impact_Sounds/Demolition/rock_newimp_12}}
 {id=WavePlayer_16, properties {soundWaveAssetPtr=/Clyde/Sounds/Fort_Impact_Sounds/Demolition/rock_newimp_01}}
 {id=WavePlayer_177, properties {soundWaveAssetPtr=&rock_impact_stereo_03}}
 {id=WavePlayer_178, properties {soundWaveAssetPtr=/Game/Sounds/Fort_Impact_Sounds/Demolition/Defaults/rock_impact_stereo_04}}
 {id=WavePlayer_180, properties {soundWaveAssetPtr=/Game/Sounds/Fort_Impact_Sounds/Demolition/Defaults/rock_impact_stereo_06}}
 {id=WavePlayer_181, properties {soundWaveAssetPtr=/Game/Sounds/Fort_Impact_Sounds/Demolition/Defaults/rock_impact_stereo_07}}
 {id=WavePlayer_185, properties {soundWaveAssetPtr=/Clyde/Sounds/Fort_Impact_Sounds/Demolition/Fort_Impact_Wobble_01}}
 {id=WavePlayer_186, properties {soundWaveAssetPtr=/Clyde/Sounds/Fort_Impact_Sounds/Demolition/Fort_Impact_Wobble_02}}
 {id=WavePlayer_187, properties {soundWaveAssetPtr=/Clyde/Sounds/Fort_Impact_Sounds/Demolition/Fort_Impact_Wobble_03}}
 {id=WavePlayer_3, properties {soundWaveAssetPtr=&rock_impact_stereo_06}}
 {id=WavePlayer_5, properties {soundWaveAssetPtr=&Fort_Impact_Wobble_01}}
 {id=WavePlayer_8, properties {soundWaveAssetPtr=&rock_newimp_03}}
 {id=WavePlayer_9, properties {soundWaveAssetPtr=&rock_newimp_11}}
]
properties {
 attenuationSettings=&Fort_Weapon_Melee_Sml_Att
 maxDistance=2800
 pitchMultiplier=1.5
 soundClassObject=&Fort_Impacts
 volumeMultiplier=0.425
 concurrencyOverrides {maxCount=3, resolutionRule=StopOldest}
 concurrencySet [&CONC_Object_Impacts_FourMax]
}
