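# Demolition sound cue for a final wood impact, likely from player hits, routed to the fort impacts sound class.
# A mixer layers three randomized, modulated branches: pitched-up wood impacts at full volume, quieter stereo impacts, and mid-level WoodFinal hits.
# The stereo layer starts after a short random delay, and each branch has its own attenuation node.
# An object-impacts concurrency set, named for a four-voice cap, limits simultaneous plays.
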SoundCue /Clyde/Sounds/Fort_Impact_Sounds/Demolition/Defaults/Demo_WoodPlayer_ImpactFinal_Cue
root=Mixer_1
nodes [
 {id=Attenuation_0, children [Modulator_2]}
 {id=Attenuation_3, children [Delay_2], properties {attenuationOverrides {falloffDistance=900, attenuationShapeExtents {x=100}}}}
 {id=Attenuation_4, children [Modulator_1], properties {attenuationOverrides {falloffDistance=900, attenuationShapeExtents {x=100}}}}
 {id=Delay_2, children [Modulator_3], properties {delayMax=0.1, delayMin=0.05}}
 {id=Mixer_1, children [Attenuation_0, Attenuation_3, Attenuation_4], properties {inputVolume [1,0.2,0.5]}}
 {id=Modulator_1, children [Random_1]}
 {id=Modulator_2, children [Random_8], properties {pitchMax=1.3, pitchMin=1.2}}
 {id=Modulator_3, children [Random_9], properties {pitchMax=1.3, pitchMin=1.1}}
 {id=Random_1, children [WavePlayer_219, WavePlayer_235, WavePlayer_236], properties {weights [1,1,1]}}
 {id=Random_8, children [WavePlayer_234, WavePlayer_233, WavePlayer_231, WavePlayer_230], properties {weights [1,1,1,1]}}
 {id=Random_9, children [WavePlayer_218, WavePlayer_224, WavePlayer_221, WavePlayer_223], properties {weights [1,1,1,1]}}
 {id=WavePlayer_218, properties {soundWaveAssetPtr=/Clyde/Sounds/Fort_Impact_Sounds/Demolition/wood_impact_stereo_01}}
 {id=WavePlayer_219, properties {soundWaveAssetPtr=/Clyde/Sounds/Fort_Impact_Sounds/Demolition/Fort_Impact_WoodFinal_01}}
 {id=WavePlayer_221, properties {soundWaveAssetPtr=/Clyde/Sounds/Fort_Impact_Sounds/Demolition/wood_impact_stereo_07}}
 {id=WavePlayer_223, properties {soundWaveAssetPtr=/Clyde/Sounds/Fort_Impact_Sounds/Demolition/wood_impact_stereo_08}}
 {id=WavePlayer_224, properties {soundWaveAssetPtr=/Clyde/Sounds/Fort_Impact_Sounds/Demolition/wood_impact_stereo_04}}
 {id=WavePlayer_230, properties {soundWaveAssetPtr=/Clyde/Sounds/Fort_Impact_Sounds/Demolition/wood_impact_05}}
 {id=WavePlayer_231, properties {soundWaveAssetPtr=/Clyde/Sounds/Fort_Impact_Sounds/Demolition/wood_impact_04}}
 {id=WavePlayer_233, properties {soundWaveAssetPtr=/Clyde/Sounds/Fort_Impact_Sounds/Demolition/wood_impact_02}}
 {id=WavePlayer_234, properties {soundWaveAssetPtr=/Clyde/Sounds/Fort_Impact_Sounds/Demolition/wood_impact_01}}
 {id=WavePlayer_235, properties {soundWaveAssetPtr=/Clyde/Sounds/Fort_Impact_Sounds/Demolition/Fort_Impact_WoodFinal_02}}
 {id=WavePlayer_236, properties {soundWaveAssetPtr=/Clyde/Sounds/Fort_Impact_Sounds/Demolition/Fort_Impact_WoodFinal_03}}
]
properties {
 maxDistance=2097152
 soundClassObject=&Fort_Impacts_PSM
 volumeMultiplier=0.425
 concurrencyOverrides {resolutionRule=PreventNew}
 concurrencySet [&CONC_Object_Impacts_FourMax]
}
