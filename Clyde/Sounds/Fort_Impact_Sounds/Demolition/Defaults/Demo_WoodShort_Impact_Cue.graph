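# Sound cue for short wooden demolition impacts, routed to the impacts sound class with small-melee attenuation.
# A quality-level node selects among four branches, whose pool of small wood impact waves shrinks from four to one in child order.
# Each branch adds slight random pitch variation, and multi-wave branches pick randomly with equal weights.
# The cue plays at reduced volume and references a shared object-impact concurrency set named for a four-voice maximum.
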SoundCue /Clyde/Sounds/Fort_Impact_Sounds/Demolition/Defaults/Demo_WoodShort_Impact_Cue
root=QualityLevel_0
nodes [
 {id=Modulator_0, children [Random_0], properties {pitchMax=1.07, pitchMin=0.92}}
 {id=Modulator_1, children [Random_1], properties {pitchMax=1.1, pitchMin=0.9}}
 {id=Modulator_2, children [Random_6], properties {pitchMax=1.07, pitchMin=0.92}}
 {id=Modulator_3, children [WavePlayer_5], properties {pitchMax=1.1, pitchMin=0.9}}
 {id=QualityLevel_0, children [Modulator_2,Modulator_0,Modulator_1,Modulator_3]}
 {id=Random_0, children [WavePlayer_0,WavePlayer_1,WavePlayer_2], properties {weights [1,1,1]}}
 {id=Random_1, children [WavePlayer_3,WavePlayer_4], properties {weights [1,1]}}
 {id=Random_6, children [WavePlayer_35, WavePlayer_36, WavePlayer_37, WavePlayer_38], properties {weights [1,1,1,1]}}
 {id=WavePlayer_0, properties {soundWaveAssetPtr=/Clyde/Sounds/Fort_Impact_Sounds/Demolition/Wood_ImpactShort_Small07}}
 {id=WavePlayer_1, properties {soundWaveAssetPtr=/Clyde/Sounds/Fort_Impact_Sounds/Demolition/Wood_ImpactShort_Small08}}
 {id=WavePlayer_2, properties {soundWaveAssetPtr=/Clyde/Sounds/Fort_Impact_Sounds/Demolition/Wood_ImpactShort_Small09}}
 {id=WavePlayer_3, properties {soundWaveAssetPtr=&Wood_ImpactShort_Small07}}
 {id=WavePlayer_35, properties {soundWaveAssetPtr=&Wood_ImpactShort_Small07}}
 {id=WavePlayer_36, properties {soundWaveAssetPtr=&Wood_ImpactShort_Small08}}
 {id=WavePlayer_37, properties {soundWaveAssetPtr=&Wood_ImpactShort_Small09}}
 {id=WavePlayer_38, properties {soundWaveAssetPtr=/Clyde/Sounds/Fort_Impact_Sounds/Demolition/Wood_ImpactShort_Small10}}
 {id=WavePlayer_4, properties {soundWaveAssetPtr=&Wood_ImpactShort_Small08}}
 {id=WavePlayer_5, properties {soundWaveAssetPtr=&Wood_ImpactShort_Small07}}
]
properties {
 attenuationSettings=&Fort_Weapon_Melee_Sml_Att
 maxDistance=2800
 soundClassObject=&Fort_Impacts
 volumeMultiplier=0.425
 concurrencyOverrides {resolutionRule=PreventNew}
 concurrencySet [&CONC_Object_Impacts_FourMax]
}
