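# Sound cue for the final impact on wooden demolition targets, routed to the Fort impacts sound class at half volume.
# A mixer layers three separately attenuated branches: pitched-down random workbench junk hits at reduced level, pitched-up random stereo wood impacts, and a wood fence final impact.
# The wood impact and fence layers start after short random delays, staggering them behind the junk hit.
# A shared four-maximum concurrency set limits simultaneous plays.
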SoundCue /Clyde/Sounds/Fort_Impact_Sounds/Demolition/Defaults/Demo_Wood_Targets_FinalImpact_Cue1
root=Mixer_1
nodes [
 {id=Attenuation_1, children [Modulator_1]}
 {id=Attenuation_2, children [Delay_0]}
 {id=Attenuation_6, children [Delay_2]}
 {id=Delay_0, children [Modulator_0], properties {delayMax=0.1, delayMin=0.05}}
 {id=Delay_2, children [Modulator_3], properties {delayMax=0.1, delayMin=0.05}}
 {id=Mixer_1, children [Attenuation_6, Attenuation_1, Attenuation_2], properties {inputVolume [1,0.5,1]}}
 {id=Modulator_0, children [Random_0]}
 {id=Modulator_1, children [Random_2], properties {pitchMax=0.95, pitchMin=0.9}}
 {id=Modulator_3, children [Random_9], properties {pitchMax=1.3, pitchMin=1.1}}
 {id=Random_0, children [WavePlayer_10], properties {weights [1]}}
 {id=Random_2, children [WavePlayer_4,WavePlayer_5,WavePlayer_6], properties {weights [1,1,1]}}
 {id=Random_9, children [WavePlayer_13, WavePlayer_91, WavePlayer_93], properties {weights [1,1,1]}}
 {id=WavePlayer_10, properties {soundWaveAssetPtr=/Clyde/Sounds/Fort_Impact_Sounds/Demolition/Wood_Fence_ImpactFinal02}}
 {id=WavePlayer_13, properties {soundWaveAssetPtr=/Clyde/Sounds/Fort_Impact_Sounds/Demolition/wood_impact_stereo_01}}
 {id=WavePlayer_4, properties {soundWaveAssetPtr=/Clyde/Sounds/Fort_Impact_Sounds/Demolition/Workbench_Junk_Hit1}}
 {id=WavePlayer_5, properties {soundWaveAssetPtr=/Clyde/Sounds/Fort_Impact_Sounds/Demolition/Workbench_Junk_Hit2}}
 {id=WavePlayer_6, properties {soundWaveAssetPtr=/Clyde/Sounds/Fort_Impact_Sounds/Demolition/Workbench_Junk_Hit3}}
 {id=WavePlayer_91, properties {soundWaveAssetPtr=/Clyde/Sounds/Fort_Impact_Sounds/Demolition/wood_impact_stereo_07}}
 {id=WavePlayer_93, properties {soundWaveAssetPtr=/Clyde/Sounds/Fort_Impact_Sounds/Demolition/wood_impact_stereo_08}}
]
properties {
 maxDistance=2097152
 soundClassObject=&Fort_Impacts_PSM
 volumeMultiplier=0.5
 concurrencyOverrides {resolutionRule=PreventNew}
 concurrencySet [/Game/Sounds/Concurrency_Rules/FourMaximum_StopFurthestOldest_VolumeScale]
}
